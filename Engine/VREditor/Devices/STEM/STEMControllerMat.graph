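# STEMControllerMat is a Default Lit material for the STEM controller device model in the VR Editor.
# It multiplies the STEM controller diffuse texture by a dark grey constant to darken it.
# A normal map sample supplies surface detail, and a lone scalar constant accompanies them, though the file does not show which material inputs these nodes feed.
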Material /Engine/VREditor/Devices/STEM/STEMControllerMat
graph {
 properties {pixelDepthOffsetMode=PDOM_Legacy, usedShadingModels=Default Lit}
 nodes [
  Constant() {r=0.4}
  Multiply(A=TextureSample() {texture=&/STEMControllerDiffuse}.RGB, B=Constant3Vector() {constant=<rgba 0.06 0.06 0.06 0>})
  TextureSample() {samplerType=SAMPLERTYPE_Normal, texture=&/STEMControllerNormal}
 ]
}
